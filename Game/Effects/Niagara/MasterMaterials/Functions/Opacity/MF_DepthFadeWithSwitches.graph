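# Material function for Niagara effect master materials that softens opacity with a depth fade only where shading path and quality level allow it, described as the non-MA version.
# It feeds the Opacity and Fade Distance inputs into a DepthFade node.
# Nested shading path and quality switches keep the faded result for deferred shading at High and Epic quality.
# Forward, mobile, Low and Medium pass the input opacity through unchanged.
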MaterialFunction /Game/Effects/Niagara/MasterMaterials/Functions/Opacity/MF_DepthFadeWithSwitches
graph {
 properties {description="Depth Fade with quality switches, non-MA version"}
 nodes [
  n1 = FunctionInput(Preview=Constant() {r=1}) {bUsePreviewValueAsDefault=true, inputName=Opacity, inputType=FunctionInput_Scalar}
  n2 = DepthFade(
   Opacity=@n1
   FadeDistance=FunctionInput(Preview=Constant() {r=50}) {bUsePreviewValueAsDefault=true, inputName=Fade Distance, inputType=FunctionInput_Scalar, sortPriority=1}
  )
  n3 = ShadingPathSwitch(Default=@n2, Deferred=@n2, Mobile=@n1) {desc=Migrated from FeatureLevelSwitch}
  n4 = QualitySwitch(Default=@n3, Low=@n1, High=@n3, Medium=@n1, Epic=@n3)
  FunctionOutput(None=ShadingPathSwitch(Default=@n4, Deferred=@n4, Forward=@n1, Mobile=@n1)) {outputName=Opacity}
 ]
}
