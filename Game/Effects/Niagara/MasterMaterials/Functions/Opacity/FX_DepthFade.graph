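# Material function for the FX Opacity library that takes material attributes and applies a depth fade to Opacity, with a parameter-driven fade distance.
# Two nested ShadingPathSwitch nodes use the faded opacity only on the Deferred path, noted as PC, while other paths keep the input opacity.
# The chosen value is written back into the material attributes for output.
# A QualitySwitch offering the fade at High quality is left unconnected to the output.
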MaterialFunction /Game/Effects/Niagara/MasterMaterials/Functions/Opacity/FX_DepthFade
graph {
 properties {description=Depth Fade with switches, libraryCategoriesText [FX Opacity]}
 nodes [
  n1 = GetMaterialAttributes(None=FunctionInput() {inputType=FunctionInput_MaterialAttributes}) {attributeGetTypes [#1]}
  n2 = DepthFade(
   Opacity=@n1.Opacity
   FadeDistance=ScalarParameter() {defaultValue=1, group="6.1) Opacity", parameterName=Dpeth Fade Distance, sortPriority=10}
  )
  n3 = ShadingPathSwitch(Default=@n1.Opacity, Deferred=@n2) {desc="Defered is pc, Forward is switch, mobile is mobile"}
  FunctionOutput(None=SetMaterialAttributes!(
    MaterialAttributes=@n1.MaterialAttributes
    Opacity=ShadingPathSwitch(Default=@n1.Opacity, Deferred=@n3, Mobile=@n1.Opacity) {desc=Migrated from FeatureLevelSwitch}
   ) {attributeSetTypes [#1]}) {outputName="Opacity (MA)"}
  QualitySwitch(Default=@n1.Opacity, High=@n2) {desc=This for PCs}
 ]
}
