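# Material function that builds opacity and interior masks for a circular FX portal defined by world position, facing direction and radius.
# A cylinder intersection mask is multiplied by a steep, clamped camera-facing term, signed between front and back.
# Offset signed distance to the portal plane forms a powered gradient, added to OpacityMask unless the CylinderIntersectionOnly static switch is on.
# InteriorMask is that gradient inverted; the MaskPosition and MaskDirection vector parameters are unconnected.
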MaterialFunction /Game/Effects/Niagara/MasterMaterials/Functions/Opacity/MF_FX_Circular_Portal_Mask
graph {nodes [
  n1 = FunctionInput() {bUsePreviewValueAsDefault=true, description=World position for the portal mask, inputName=MaskPosition}
  n2 = FunctionInput() {
   bUsePreviewValueAsDefault=true
   description=Radius for the circular portal mask
   inputName=MaskRadius
   inputType=FunctionInput_Scalar
   sortPriority=2
   previewValue=<1000 0 0 0>
  }
  n3 = FunctionInput() {bUsePreviewValueAsDefault=true, description=Direction the portal is facing, inputName=MaskDirection, sortPriority=1, previewValue=<1 0 0 0>}
  n4 = Normalize(VectorInput=@n3)
  n5 = DotProduct(A=@n4, B=CameraVectorWS())
  n6 = Multiply(
   A=MaterialFunctionCall!(Origin=@n1, Radius=@n2, Z Axis=@n4, XY Scale, X Axis, Capped?, Z Min-Max) {materialFunction=/Engine/Functions/Engine_MaterialFunctions01/Math/CylinderIntersection}.Opacity Mask
   B=Clamp(None=Multiply(A=@n5) {constB=10}) {minDefault=-1}
  )
  n7 = WorldPosition()
  n8 = FunctionInput() {
   bUsePreviewValueAsDefault=true
   description=Offset the intersect fade inside or outside the portal
   inputName=MaskEdgeGradientOffset
   inputType=FunctionInput_Scalar
   sortPriority=4
  }
  n9 = Add(A=DotProduct(A=Subtract(A=@n7.XYZ, B=@n1), B=@n4), B=@n8)
  n10 = FunctionInput() {
   bUsePreviewValueAsDefault=true
   description=Fade out at the intersection with the portal mask plane
   inputName=MaskEdgeGradientSize
   inputType=FunctionInput_Scalar
   sortPriority=3
   previewValue=<100 0 0 0>
  }
  n11 = FunctionInput() {bUsePreviewValueAsDefault=true, inputName=MaskEdgeGradientPower, inputType=FunctionInput_Scalar, sortPriority=5, previewValue=<2 0 0 1>}
  n12 = Power(Base=Saturate(None=Divide(A=@n9, B=@n10)), Exp=@n11)
  FunctionOutput(None=StaticSwitchParameter(True=@n6, False=Add(A=@n12, B=@n6)) {group=PortalMask, parameterName=CylinderIntersectionOnly}) {outputName=OpacityMask}
  FunctionOutput(None=OneMinus(None=@n12)) {outputName=InteriorMask, sortPriority=1}
  VectorParameter() {parameterName=MaskPosition, defaultValue=<rgba -102571.1094 -95641.6406 6059.9902 0>}
  VectorParameter() {parameterName=MaskDirection, defaultValue=<rgba 1 0 -0.25 0>}
 ]}
