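# Material function in the FX Opacity library that bevels particle ribbon opacity, rounding the base of the ribbon.
# A parabolic profile across UV0's G axis is cut by a squared falloff of particle relative time, scaled by Ribbon Rounding.
# Ribbon Rounding Softness sets the edge width, which widens toward the center line.
# The mask multiplies incoming Opacity; a BaseColor multiply exists but does not feed the output.
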MaterialFunction /Game/Effects/Niagara/MasterMaterials/Functions/Opacity/MF_Parametric_OpacityBevel
graph {
 properties {bExposeToLibrary=true, description=Fades by parametric parameters., libraryCategoriesText [FX Opacity]}
 nodes [
  n1 = GetMaterialAttributes(None=FunctionInput() {inputType=FunctionInput_MaterialAttributes}) {attributeGetTypes [#1, #2, #3]}
  n2 = TextureCoordinate()
  n3 = Multiply(A=Subtract(A=ComponentMask(None=@n2) {desc=UV0 G seems broke., g=true}, B=Constant() {r=0.5}), B=Constant() {r=2})
  n4 = OneMinus(None=Saturate(None=Multiply(A=@n3, B=@n3)))
  n5 = ParticleRelativeTime()
  n6 = ScalarParameter() {
   defaultValue=0.5
   desc=Amount of rounding on the base of the ribbon
   group="6.1) Opacity"
   parameterName=Ribbon Rounding
   sliderMax=1
   sortPriority=1
  }
  n7 = OneMinus(None=Saturate(None=Divide(A=@n5, B=@n6)))
  n8 = Multiply(A=@n7, B=@n7)
  n9 = ScalarParameter() {defaultValue=0.2, desc=Softness, group="6.1) Opacity", parameterName=Ribbon Rounding Softness, sliderMax=1, sortPriority=2}
  n10 = OneMinus(None=Abs(None=@n3))
  n11 = Multiply(A=@n10, B=@n10)
  n12 = Saturate(None=Divide(A=Subtract(A=@n4, B=@n8), B=LinearInterpolate(A=@n9, Alpha=@n11)))
  n13 = SetMaterialAttributes!(MaterialAttributes=@n1.MaterialAttributes, Opacity=Multiply(A=@n1.Opacity, B=@n12)) {attributeSetTypes [#1]}
  FunctionOutput(None=@n13) {outputName="Opacitys and BaseColor (MA)"}
  Multiply(B=@n7)
  n14 = GetMaterialAttributes(None=@n13) {attributeGetTypes [#4]}
  SetMaterialAttributes!(MaterialAttributes=@n14.MaterialAttributes, BaseColor=Multiply(A=@n14.BaseColor, B=@n12)) {attributeSetTypes [#4]}
 ]
}
